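# Sound cue for the NitroDrome PA announcer calling out a boss eliminating a player, backed by a crowd cheer.
# A mixer layers a random announcer line at full volume with a quieter random crowd cheer that starts after a fixed two-second delay.
# It plays on the NPC VO sound class with announcer attenuation and concurrency settings, and both random pools are excluded from branch culling.
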SoundCue /NitroDrome/Assets/Audio/PA/VO/PA_Announcer/VO_NitroDrome_PA_Boss_Elims_Player_Cue
root=Mixer_0
nodes [
 {id=Delay_0, children [Random_0], properties {delayMax=2, delayMin=2}}
 {id=Mixer_0, children [Random_1,Delay_0], properties {inputVolume [1,0.6]}}
 {
  id=Random_0
  children [WavePlayer_0, WavePlayer_1, WavePlayer_2, WavePlayer_3, WavePlayer_4, WavePlayer_10, WavePlayer_11, WavePlayer_5]
  properties {bShouldExcludeFromBranchCulling=true, weights [1,1,1,1,1,1,1,1]}
 }
 {
  id=Random_1
  children [WavePlayer_6, WavePlayer_7, WavePlayer_8, WavePlayer_9, WavePlayer_17, WavePlayer_18]
  properties {bShouldExcludeFromBranchCulling=true, weights [1,1,1,1,1,1]}
 }
 {id=WavePlayer_0, properties {soundWaveAssetPtr=/NitroDrome/Assets/Audio/PA/Crowd/NitroDrome_PA_Crowd_Cheer_01}}
 {id=WavePlayer_1, properties {soundWaveAssetPtr=/NitroDrome/Assets/Audio/PA/Crowd/NitroDrome_PA_Crowd_Cheer_02}}
 {id=WavePlayer_10, properties {soundWaveAssetPtr=/NitroDrome/Assets/Audio/PA/Crowd/NitroDrome_PA_Crowd_Cheer_06}}
 {id=WavePlayer_11, properties {soundWaveAssetPtr=/NitroDrome/Assets/Audio/PA/Crowd/NitroDrome_PA_Crowd_Cheer_07}}
 {id=WavePlayer_17, properties {soundWaveAssetPtr=&/VO_NitroDome_PA_Boss_Elims_Player_05}}
 {id=WavePlayer_18, properties {soundWaveAssetPtr=&/VO_NitroDome_PA_Boss_Elims_Player_06}}
 {id=WavePlayer_2, properties {soundWaveAssetPtr=/NitroDrome/Assets/Audio/PA/Crowd/NitroDrome_PA_Crowd_Cheer_03}}
 {id=WavePlayer_3, properties {soundWaveAssetPtr=/NitroDrome/Assets/Audio/PA/Crowd/NitroDrome_PA_Crowd_Cheer_04}}
 {id=WavePlayer_4, properties {soundWaveAssetPtr=/NitroDrome/Assets/Audio/PA/Crowd/NitroDrome_PA_Crowd_Cheer_05}}
 {id=WavePlayer_5, properties {soundWaveAssetPtr=/NitroDrome/Assets/Audio/PA/Crowd/NitroDrome_PA_Crowd_Cheer_08}}
 {id=WavePlayer_6, properties {soundWaveAssetPtr=&/VO_NitroDome_PA_Boss_Elims_Player_01}}
 {id=WavePlayer_7, properties {soundWaveAssetPtr=&/VO_NitroDome_PA_Boss_Elims_Player_02}}
 {id=WavePlayer_8, properties {soundWaveAssetPtr=&/VO_NitroDome_PA_Boss_Elims_Player_03}}
 {id=WavePlayer_9, properties {soundWaveAssetPtr=&/VO_NitroDome_PA_Boss_Elims_Player_04}}
]
properties {
 attenuationSettings=/NitroDrome/Assets/Audio/PA/VO/VO_PA_Announcer_Att
 maxDistance=20000
 soundClassObject=&Fort_VO_NPC
 volumeMultiplier=0.85
 concurrencySet [/NitroDrome/Assets/Audio/PA/VO/VO_PA_Announcer_Conc]
}
