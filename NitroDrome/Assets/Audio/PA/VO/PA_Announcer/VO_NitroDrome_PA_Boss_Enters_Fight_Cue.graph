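# NitroDrome PA announcer sound cue for a boss entering the fight, layering a voice line with delayed crowd reactions.
# A mixer combines a random pick from three announcer lines at full volume with two quieter crowd-cheer layers.
# Each crowd layer randomly picks from its own four cheers, one after a fixed delay and the other after a slightly longer, randomized delay.
# Announcer-specific attenuation and concurrency apply, under the NPC voice sound class.
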SoundCue /NitroDrome/Assets/Audio/PA/VO/PA_Announcer/VO_NitroDrome_PA_Boss_Enters_Fight_Cue
root=Mixer_0
nodes [
 {id=Delay_1, children [Random_4], properties {delayMax=2, delayMin=2}}
 {id=Delay_2, children [Random_5], properties {delayMax=3.5, delayMin=2.75}}
 {id=Mixer_0, children [Random_3,Delay_1,Delay_2], properties {inputVolume [1,0.6,0.7]}}
 {id=Random_3, children [WavePlayer_14, WavePlayer_15, WavePlayer_19], properties {bShouldExcludeFromBranchCulling=true, weights [1,1,1]}}
 {
  id=Random_4
  children [WavePlayer_20, WavePlayer_21, WavePlayer_22, WavePlayer_23]
  properties {bShouldExcludeFromBranchCulling=true, weights [1,1,1,1]}
 }
 {
  id=Random_5
  children [WavePlayer_25, WavePlayer_26, WavePlayer_27, WavePlayer_28]
  properties {bShouldExcludeFromBranchCulling=true, weights [1,1,1,1]}
 }
 {id=WavePlayer_14, properties {soundWaveAssetPtr=&/VO_NitroDome_PA_Boss_Enters_Fight_01}}
 {id=WavePlayer_15, properties {soundWaveAssetPtr=&/VO_NitroDome_PA_Boss_Enters_Fight_02}}
 {id=WavePlayer_19, properties {soundWaveAssetPtr=&/VO_NitroDome_PA_Boss_Enters_Fight_06}}
 {id=WavePlayer_20, properties {soundWaveAssetPtr=/NitroDrome/Assets/Audio/PA/Crowd/NitroDrome_PA_Crowd_Cheer_01}}
 {id=WavePlayer_21, properties {soundWaveAssetPtr=/NitroDrome/Assets/Audio/PA/Crowd/NitroDrome_PA_Crowd_Cheer_02}}
 {id=WavePlayer_22, properties {soundWaveAssetPtr=/NitroDrome/Assets/Audio/PA/Crowd/NitroDrome_PA_Crowd_Cheer_03}}
 {id=WavePlayer_23, properties {soundWaveAssetPtr=/NitroDrome/Assets/Audio/PA/Crowd/NitroDrome_PA_Crowd_Cheer_04}}
 {id=WavePlayer_25, properties {soundWaveAssetPtr=/NitroDrome/Assets/Audio/PA/Crowd/NitroDrome_PA_Crowd_Cheer_05}}
 {id=WavePlayer_26, properties {soundWaveAssetPtr=/NitroDrome/Assets/Audio/PA/Crowd/NitroDrome_PA_Crowd_Cheer_06}}
 {id=WavePlayer_27, properties {soundWaveAssetPtr=/NitroDrome/Assets/Audio/PA/Crowd/NitroDrome_PA_Crowd_Cheer_07}}
 {id=WavePlayer_28, properties {soundWaveAssetPtr=/NitroDrome/Assets/Audio/PA/Crowd/NitroDrome_PA_Crowd_Cheer_08}}
]
properties {
 attenuationSettings=/NitroDrome/Assets/Audio/PA/VO/VO_PA_Announcer_Att
 maxDistance=20000
 soundClassObject=&Fort_VO_NPC
 volumeMultiplier=0.85
 concurrencySet [/NitroDrome/Assets/Audio/PA/VO/VO_PA_Announcer_Conc]
}
