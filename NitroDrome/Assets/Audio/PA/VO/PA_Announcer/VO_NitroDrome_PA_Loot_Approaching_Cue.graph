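# Sound cue for the NitroDrome PA announcer's loot-approaching callout, layering a voice line over a crowd cheer.
# A Mixer plays a random announcer line at full volume alongside a random crowd cheer at half volume, delayed by a fixed interval.
# The voice waves are named Loot_Aquired, although the cue is named Loot_Approaching.
# Announcer attenuation and concurrency assets and the NPC VO sound class apply to the whole cue, crowd included.
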SoundCue /NitroDrome/Assets/Audio/PA/VO/PA_Announcer/VO_NitroDrome_PA_Loot_Approaching_Cue
root=Mixer_0
nodes [
 {id=Delay_0, children [Random_0], properties {delayMax=1.5, delayMin=1.5}}
 {id=Mixer_0, children [Random_2,Delay_0], properties {inputVolume [1,0.5]}}
 {
  id=Random_0
  children [WavePlayer_0, WavePlayer_1, WavePlayer_2, WavePlayer_3, WavePlayer_4, WavePlayer_10, WavePlayer_11, WavePlayer_12]
  properties {bShouldExcludeFromBranchCulling=true, weights [1,1,1,1,1,1,1,1]}
 }
 {id=Random_2, children [WavePlayer_5, WavePlayer_13, WavePlayer_14, WavePlayer_15], properties {weights [1,1,1,1]}}
 {id=WavePlayer_0, properties {soundWaveAssetPtr=/NitroDrome/Assets/Audio/PA/Crowd/NitroDrome_PA_Crowd_Cheer_01}}
 {id=WavePlayer_1, properties {soundWaveAssetPtr=/NitroDrome/Assets/Audio/PA/Crowd/NitroDrome_PA_Crowd_Cheer_02}}
 {id=WavePlayer_10, properties {soundWaveAssetPtr=/NitroDrome/Assets/Audio/PA/Crowd/NitroDrome_PA_Crowd_Cheer_06}}
 {id=WavePlayer_11, properties {soundWaveAssetPtr=/NitroDrome/Assets/Audio/PA/Crowd/NitroDrome_PA_Crowd_Cheer_07}}
 {id=WavePlayer_12, properties {soundWaveAssetPtr=/NitroDrome/Assets/Audio/PA/Crowd/NitroDrome_PA_Crowd_Cheer_08}}
 {id=WavePlayer_13, properties {soundWaveAssetPtr=&/VO_NitroDome_PA_Loot_Aquired_02}}
 {id=WavePlayer_14, properties {soundWaveAssetPtr=&/VO_NitroDome_PA_Loot_Aquired_03}}
 {id=WavePlayer_15, properties {soundWaveAssetPtr=&/VO_NitroDome_PA_Loot_Aquired_04}}
 {id=WavePlayer_2, properties {soundWaveAssetPtr=/NitroDrome/Assets/Audio/PA/Crowd/NitroDrome_PA_Crowd_Cheer_03}}
 {id=WavePlayer_3, properties {soundWaveAssetPtr=/NitroDrome/Assets/Audio/PA/Crowd/NitroDrome_PA_Crowd_Cheer_04}}
 {id=WavePlayer_4, properties {soundWaveAssetPtr=/NitroDrome/Assets/Audio/PA/Crowd/NitroDrome_PA_Crowd_Cheer_05}}
 {id=WavePlayer_5, properties {soundWaveAssetPtr=&/VO_NitroDome_PA_Loot_Aquired_01}}
]
properties {
 attenuationSettings=/NitroDrome/Assets/Audio/PA/VO/VO_PA_Announcer_Att
 maxDistance=20000
 soundClassObject=&Fort_VO_NPC
 volumeMultiplier=0.85
 concurrencySet [/NitroDrome/Assets/Audio/PA/VO/VO_PA_Announcer_Conc]
}
