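# Sound cue for the NitroDrome PA announcer's Nitro Geyser callout, layered with a crowd reaction.
# A mixer plays a random announcer line alongside a quieter crowd branch, which randomly picks either a cheer set or a shock set, each behind a delay node.
# Announcer, cheer and shock randomizers are excluded from branch culling, and the cue uses NPC VO sound class with PA announcer attenuation and concurrency.
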SoundCue /NitroDrome/Assets/Audio/PA/VO/PA_Announcer/VO_NitroDrome_PA_Nitro_Geyser_Cue
root=Mixer_0
nodes [
 {id=Delay_0, children [Random_2]}
 {id=Delay_1, children [Random_3]}
 {id=Mixer_0, children [Random_0,Random_4], properties {inputVolume [1,0.6]}}
 {
  id=Random_0
  children [WavePlayer_0, WavePlayer_1, WavePlayer_2, WavePlayer_3, WavePlayer_13]
  properties {bShouldExcludeFromBranchCulling=true, weights [1,1,1,1,1]}
 }
 {
  id=Random_2
  children [WavePlayer_14, WavePlayer_15, WavePlayer_16, WavePlayer_17, WavePlayer_18, WavePlayer_10, WavePlayer_11, WavePlayer_12]
  properties {bShouldExcludeFromBranchCulling=true, weights [1,1,1,1,1,1,1,1]}
 }
 {
  id=Random_3
  children [WavePlayer_22, WavePlayer_23, WavePlayer_24, WavePlayer_25, WavePlayer_26, WavePlayer_27]
  properties {bShouldExcludeFromBranchCulling=true, weights [1,1,1,1,1,1]}
 }
 {id=Random_4, children [Delay_0,Delay_1], properties {weights [1,1]}}
 {id=WavePlayer_0, properties {soundWaveAssetPtr=&/VO_NitroDome_PA_Nitro_Geyser_01}}
 {id=WavePlayer_1, properties {soundWaveAssetPtr=&/VO_NitroDome_PA_Nitro_Geyser_02}}
 {id=WavePlayer_10, properties {soundWaveAssetPtr=/NitroDrome/Assets/Audio/PA/Crowd/NitroDrome_PA_Crowd_Cheer_06}}
 {id=WavePlayer_11, properties {soundWaveAssetPtr=/NitroDrome/Assets/Audio/PA/Crowd/NitroDrome_PA_Crowd_Cheer_07}}
 {id=WavePlayer_12, properties {soundWaveAssetPtr=/NitroDrome/Assets/Audio/PA/Crowd/NitroDrome_PA_Crowd_Cheer_08}}
 {id=WavePlayer_13, properties {soundWaveAssetPtr=&/VO_NitroDome_PA_Nitro_Geyser_06}}
 {id=WavePlayer_14, properties {soundWaveAssetPtr=/NitroDrome/Assets/Audio/PA/Crowd/NitroDrome_PA_Crowd_Cheer_01}}
 {id=WavePlayer_15, properties {soundWaveAssetPtr=/NitroDrome/Assets/Audio/PA/Crowd/NitroDrome_PA_Crowd_Cheer_02}}
 {id=WavePlayer_16, properties {soundWaveAssetPtr=/NitroDrome/Assets/Audio/PA/Crowd/NitroDrome_PA_Crowd_Cheer_03}}
 {id=WavePlayer_17, properties {soundWaveAssetPtr=/NitroDrome/Assets/Audio/PA/Crowd/NitroDrome_PA_Crowd_Cheer_04}}
 {id=WavePlayer_18, properties {soundWaveAssetPtr=/NitroDrome/Assets/Audio/PA/Crowd/NitroDrome_PA_Crowd_Cheer_05}}
 {id=WavePlayer_2, properties {soundWaveAssetPtr=&/VO_NitroDome_PA_Nitro_Geyser_03}}
 {id=WavePlayer_22, properties {soundWaveAssetPtr=/NitroDrome/Assets/Audio/PA/Crowd/NitroDrome_PA_Crowd_Shock_01}}
 {id=WavePlayer_23, properties {soundWaveAssetPtr=/NitroDrome/Assets/Audio/PA/Crowd/NitroDrome_PA_Crowd_Shock_02}}
 {id=WavePlayer_24, properties {soundWaveAssetPtr=/NitroDrome/Assets/Audio/PA/Crowd/NitroDrome_PA_Crowd_Shock_03}}
 {id=WavePlayer_25, properties {soundWaveAssetPtr=/NitroDrome/Assets/Audio/PA/Crowd/NitroDrome_PA_Crowd_Shock_04}}
 {id=WavePlayer_26, properties {soundWaveAssetPtr=/NitroDrome/Assets/Audio/PA/Crowd/NitroDrome_PA_Crowd_Shock_05}}
 {id=WavePlayer_27, properties {soundWaveAssetPtr=/NitroDrome/Assets/Audio/PA/Crowd/NitroDrome_PA_Crowd_Shock_06}}
 {id=WavePlayer_3, properties {soundWaveAssetPtr=&/VO_NitroDome_PA_Nitro_Geyser_04}}
]
properties {
 attenuationSettings=/NitroDrome/Assets/Audio/PA/VO/VO_PA_Announcer_Att
 maxDistance=20000
 soundClassObject=&Fort_VO_NPC
 volumeMultiplier=0.85
 concurrencySet [/NitroDrome/Assets/Audio/PA/VO/VO_PA_Announcer_Conc]
}
