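# NitroDrome PA announcer Sound Cue for a player elimination, pairing a random voice line with a crowd reaction.
# A Mixer layers one announcer line, drawn evenly from knock-target and elimination takes, over a quieter crowd branch.
# That branch picks evenly between cheer and shock groups, each playing a random clip through a Delay node left at default settings.
# The cue uses the NPC voice sound class with announcer-specific attenuation and concurrency.
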SoundCue /NitroDrome/Assets/Audio/PA/VO/PA_Announcer/VO_NitroDrome_PA_Player_Eliminated_Cue
root=Mixer_1
nodes [
 {id=Delay_1, children [Random_6]}
 {id=Delay_3, children [Random_5]}
 {id=Mixer_1, children [Random_4,Random_7], properties {inputVolume [1,0.6]}}
 {
  id=Random_4
  children [
   WavePlayer_5
   WavePlayer_6
   WavePlayer_7
   WavePlayer_8
   WavePlayer_12
   WavePlayer_20
   WavePlayer_21
   WavePlayer_22
   WavePlayer_23
   WavePlayer_24
   WavePlayer_25
  ]
  properties {bShouldExcludeFromBranchCulling=true, weights [1,1,1,1,1,1,1,1,1,1,1]}
 }
 {
  id=Random_5
  children [WavePlayer_32, WavePlayer_33, WavePlayer_34, WavePlayer_35, WavePlayer_36, WavePlayer_37, WavePlayer_38, WavePlayer_39]
  properties {bShouldExcludeFromBranchCulling=true, weights [1,1,1,1,1,1,1,1]}
 }
 {
  id=Random_6
  children [WavePlayer_26, WavePlayer_27, WavePlayer_28, WavePlayer_29, WavePlayer_30, WavePlayer_31]
  properties {bShouldExcludeFromBranchCulling=true, weights [1,1,1,1,1,1]}
 }
 {id=Random_7, children [Delay_3,Delay_1], properties {weights [1,1]}}
 {id=WavePlayer_12, properties {soundWaveAssetPtr=&/VO_NitroDome_PA_Player_Knocks_Target_05}}
 {id=WavePlayer_20, properties {soundWaveAssetPtr=&/VO_NitroDome_PA_Player_Knocks_Target_06}}
 {id=WavePlayer_21, properties {soundWaveAssetPtr=&/VO_NitroDome_PA_Player_Elim_01}}
 {id=WavePlayer_22, properties {soundWaveAssetPtr=&/VO_NitroDome_PA_Player_Elim_02}}
 {id=WavePlayer_23, properties {soundWaveAssetPtr=&/VO_NitroDome_PA_Player_Elim_03}}
 {id=WavePlayer_24, properties {soundWaveAssetPtr=&/VO_NitroDome_PA_Player_Elim_04}}
 {id=WavePlayer_25, properties {soundWaveAssetPtr=&/VO_NitroDome_PA_Player_Elim_05}}
 {id=WavePlayer_26, properties {soundWaveAssetPtr=/NitroDrome/Assets/Audio/PA/Crowd/NitroDrome_PA_Crowd_Shock_01}}
 {id=WavePlayer_27, properties {soundWaveAssetPtr=/NitroDrome/Assets/Audio/PA/Crowd/NitroDrome_PA_Crowd_Shock_02}}
 {id=WavePlayer_28, properties {soundWaveAssetPtr=/NitroDrome/Assets/Audio/PA/Crowd/NitroDrome_PA_Crowd_Shock_03}}
 {id=WavePlayer_29, properties {soundWaveAssetPtr=/NitroDrome/Assets/Audio/PA/Crowd/NitroDrome_PA_Crowd_Shock_04}}
 {id=WavePlayer_30, properties {soundWaveAssetPtr=/NitroDrome/Assets/Audio/PA/Crowd/NitroDrome_PA_Crowd_Shock_05}}
 {id=WavePlayer_31, properties {soundWaveAssetPtr=/NitroDrome/Assets/Audio/PA/Crowd/NitroDrome_PA_Crowd_Shock_06}}
 {id=WavePlayer_32, properties {soundWaveAssetPtr=/NitroDrome/Assets/Audio/PA/Crowd/NitroDrome_PA_Crowd_Cheer_01}}
 {id=WavePlayer_33, properties {soundWaveAssetPtr=/NitroDrome/Assets/Audio/PA/Crowd/NitroDrome_PA_Crowd_Cheer_02}}
 {id=WavePlayer_34, properties {soundWaveAssetPtr=/NitroDrome/Assets/Audio/PA/Crowd/NitroDrome_PA_Crowd_Cheer_03}}
 {id=WavePlayer_35, properties {soundWaveAssetPtr=/NitroDrome/Assets/Audio/PA/Crowd/NitroDrome_PA_Crowd_Cheer_04}}
 {id=WavePlayer_36, properties {soundWaveAssetPtr=/NitroDrome/Assets/Audio/PA/Crowd/NitroDrome_PA_Crowd_Cheer_05}}
 {id=WavePlayer_37, properties {soundWaveAssetPtr=/NitroDrome/Assets/Audio/PA/Crowd/NitroDrome_PA_Crowd_Cheer_06}}
 {id=WavePlayer_38, properties {soundWaveAssetPtr=/NitroDrome/Assets/Audio/PA/Crowd/NitroDrome_PA_Crowd_Cheer_07}}
 {id=WavePlayer_39, properties {soundWaveAssetPtr=/NitroDrome/Assets/Audio/PA/Crowd/NitroDrome_PA_Crowd_Cheer_08}}
 {id=WavePlayer_5, properties {soundWaveAssetPtr=&/VO_NitroDome_PA_Player_Knocks_Target_01}}
 {id=WavePlayer_6, properties {soundWaveAssetPtr=&/VO_NitroDome_PA_Player_Knocks_Target_02}}
 {id=WavePlayer_7, properties {soundWaveAssetPtr=&/VO_NitroDome_PA_Player_Knocks_Target_03}}
 {id=WavePlayer_8, properties {soundWaveAssetPtr=&/VO_NitroDome_PA_Player_Knocks_Target_04}}
]
properties {
 attenuationSettings=/NitroDrome/Assets/Audio/PA/VO/VO_PA_Announcer_Att
 maxDistance=20000
 soundClassObject=&Fort_VO_NPC
 volumeMultiplier=0.85
 concurrencySet [/NitroDrome/Assets/Audio/PA/VO/VO_PA_Announcer_Conc]
}
