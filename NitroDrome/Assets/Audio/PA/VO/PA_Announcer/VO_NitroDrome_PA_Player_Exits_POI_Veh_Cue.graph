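# NitroDrome PA announcer SoundCue, named for a player exiting the POI in or from a vehicle, pairing commentary with crowd reaction.
# A Mixer layers one of three random announcer lines with a quieter crowd branch.
# That branch randomly picks jeers or shocked reactions, each a random variation pool behind a Delay node left at default timing.
# Shared announcer attenuation and concurrency assets govern playback under an NPC VO sound class.
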SoundCue /NitroDrome/Assets/Audio/PA/VO/PA_Announcer/VO_NitroDrome_PA_Player_Exits_POI_Veh_Cue
root=Mixer_0
nodes [
 {id=Delay_0, children [Random_0]}
 {id=Delay_1, children [Random_3]}
 {id=Mixer_0, children [Random_2,Random_4], properties {inputVolume [1,0.6]}}
 {
  id=Random_0
  children [WavePlayer_0, WavePlayer_1, WavePlayer_2, WavePlayer_3, WavePlayer_9, WavePlayer_13, WavePlayer_19, WavePlayer_20]
  properties {bShouldExcludeFromBranchCulling=true, weights [1,1,1,1,1,1,1,1]}
 }
 {id=Random_2, children [WavePlayer_10, WavePlayer_11, WavePlayer_12], properties {bShouldExcludeFromBranchCulling=true, weights [1,1,1]}}
 {
  id=Random_3
  children [WavePlayer_22, WavePlayer_23, WavePlayer_24, WavePlayer_25, WavePlayer_26, WavePlayer_27]
  properties {bShouldExcludeFromBranchCulling=true, weights [1,1,1,1,1,1]}
 }
 {id=Random_4, children [Delay_0,Delay_1], properties {weights [1,1]}}
 {id=WavePlayer_0, properties {soundWaveAssetPtr=/NitroDrome/Assets/Audio/PA/Crowd/NitroDrome_PA_Crowd_Jeer_01}}
 {id=WavePlayer_1, properties {soundWaveAssetPtr=/NitroDrome/Assets/Audio/PA/Crowd/NitroDrome_PA_Crowd_Jeer_02}}
 {id=WavePlayer_10, properties {soundWaveAssetPtr=&/VO_NitroDome_PA_Player_Exits_POI_Veh_01}}
 {id=WavePlayer_11, properties {soundWaveAssetPtr=&/VO_NitroDome_PA_Player_Exits_POI_Veh_02}}
 {id=WavePlayer_12, properties {soundWaveAssetPtr=&/VO_NitroDome_PA_Player_Exits_POI_Veh_03}}
 {id=WavePlayer_13, properties {soundWaveAssetPtr=/NitroDrome/Assets/Audio/PA/Crowd/NitroDrome_PA_Crowd_Jeer_06}}
 {id=WavePlayer_19, properties {soundWaveAssetPtr=/NitroDrome/Assets/Audio/PA/Crowd/NitroDrome_PA_Crowd_Jeer_07}}
 {id=WavePlayer_2, properties {soundWaveAssetPtr=/NitroDrome/Assets/Audio/PA/Crowd/NitroDrome_PA_Crowd_Jeer_03}}
 {id=WavePlayer_20, properties {soundWaveAssetPtr=/NitroDrome/Assets/Audio/PA/Crowd/NitroDrome_PA_Crowd_Jeer_08}}
 {id=WavePlayer_22, properties {soundWaveAssetPtr=/NitroDrome/Assets/Audio/PA/Crowd/NitroDrome_PA_Crowd_Shock_01}}
 {id=WavePlayer_23, properties {soundWaveAssetPtr=/NitroDrome/Assets/Audio/PA/Crowd/NitroDrome_PA_Crowd_Shock_02}}
 {id=WavePlayer_24, properties {soundWaveAssetPtr=/NitroDrome/Assets/Audio/PA/Crowd/NitroDrome_PA_Crowd_Shock_03}}
 {id=WavePlayer_25, properties {soundWaveAssetPtr=/NitroDrome/Assets/Audio/PA/Crowd/NitroDrome_PA_Crowd_Shock_04}}
 {id=WavePlayer_26, properties {soundWaveAssetPtr=/NitroDrome/Assets/Audio/PA/Crowd/NitroDrome_PA_Crowd_Shock_05}}
 {id=WavePlayer_27, properties {soundWaveAssetPtr=/NitroDrome/Assets/Audio/PA/Crowd/NitroDrome_PA_Crowd_Shock_06}}
 {id=WavePlayer_3, properties {soundWaveAssetPtr=/NitroDrome/Assets/Audio/PA/Crowd/NitroDrome_PA_Crowd_Jeer_04}}
 {id=WavePlayer_9, properties {soundWaveAssetPtr=/NitroDrome/Assets/Audio/PA/Crowd/NitroDrome_PA_Crowd_Jeer_05}}
]
properties {
 attenuationSettings=/NitroDrome/Assets/Audio/PA/VO/VO_PA_Announcer_Att
 maxDistance=20000
 soundClassObject=&Fort_VO_NPC
 volumeMultiplier=0.85
 concurrencySet [/NitroDrome/Assets/Audio/PA/VO/VO_PA_Announcer_Conc]
}
