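# NitroDrome PA announcer Sound Cue for player knock events, pairing a voice line with a crowd reaction.
# A Mixer layers a random pick from knock and elimination lines with a quieter crowd branch.
# The crowd branch randomly chooses a cheer or shock pool, each behind a Delay node left at default settings.
# Announcer attenuation and concurrency assets govern the whole cue, including the crowd, under the NPC VO sound class.
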SoundCue /NitroDrome/Assets/Audio/PA/VO/PA_Announcer/VO_NitroDrome_PA_Player_Knocked_Cue
root=Mixer_1
nodes [
 {id=Delay_0, children [Random_6]}
 {id=Delay_2, children [Random_7]}
 {id=Mixer_1, children [Random_2,Random_8], properties {inputVolume [1,0.6]}}
 {
  id=Random_2
  children [
   WavePlayer_15
   WavePlayer_16
   WavePlayer_17
   WavePlayer_18
   WavePlayer_19
   WavePlayer_20
   WavePlayer_21
   WavePlayer_28
   WavePlayer_29
   WavePlayer_30
   WavePlayer_31
  ]
  properties {bShouldExcludeFromBranchCulling=true, weights [1,1,1,1,1,1,1,1,1,1,1]}
 }
 {
  id=Random_6
  children [WavePlayer_40, WavePlayer_41, WavePlayer_42, WavePlayer_43, WavePlayer_44, WavePlayer_45, WavePlayer_46, WavePlayer_47]
  properties {bShouldExcludeFromBranchCulling=true, weights [1,1,1,1,1,1,1,1]}
 }
 {
  id=Random_7
  children [WavePlayer_48, WavePlayer_49, WavePlayer_50, WavePlayer_51, WavePlayer_52, WavePlayer_53]
  properties {bShouldExcludeFromBranchCulling=true, weights [1,1,1,1,1,1]}
 }
 {id=Random_8, children [Delay_0,Delay_2], properties {weights [1,1]}}
 {id=WavePlayer_15, properties {soundWaveAssetPtr=&/VO_NitroDome_PA_Player_Knocks_Target_01}}
 {id=WavePlayer_16, properties {soundWaveAssetPtr=&/VO_NitroDome_PA_Player_Knocks_Target_02}}
 {id=WavePlayer_17, properties {soundWaveAssetPtr=&/VO_NitroDome_PA_Player_Knocks_Target_03}}
 {id=WavePlayer_18, properties {soundWaveAssetPtr=&/VO_NitroDome_PA_Player_Knocks_Target_04}}
 {id=WavePlayer_19, properties {soundWaveAssetPtr=&/VO_NitroDome_PA_Player_Knocks_Target_05}}
 {id=WavePlayer_20, properties {soundWaveAssetPtr=&/VO_NitroDome_PA_Player_Knocks_Target_06}}
 {id=WavePlayer_21, properties {soundWaveAssetPtr=&/VO_NitroDome_PA_Player_Elim_01}}
 {id=WavePlayer_28, properties {soundWaveAssetPtr=&/VO_NitroDome_PA_Player_Elim_02}}
 {id=WavePlayer_29, properties {soundWaveAssetPtr=&/VO_NitroDome_PA_Player_Elim_03}}
 {id=WavePlayer_30, properties {soundWaveAssetPtr=&/VO_NitroDome_PA_Player_Elim_04}}
 {id=WavePlayer_31, properties {soundWaveAssetPtr=&/VO_NitroDome_PA_Player_Elim_05}}
 {id=WavePlayer_40, properties {soundWaveAssetPtr=/NitroDrome/Assets/Audio/PA/Crowd/NitroDrome_PA_Crowd_Cheer_01}}
 {id=WavePlayer_41, properties {soundWaveAssetPtr=/NitroDrome/Assets/Audio/PA/Crowd/NitroDrome_PA_Crowd_Cheer_02}}
 {id=WavePlayer_42, properties {soundWaveAssetPtr=/NitroDrome/Assets/Audio/PA/Crowd/NitroDrome_PA_Crowd_Cheer_03}}
 {id=WavePlayer_43, properties {soundWaveAssetPtr=/NitroDrome/Assets/Audio/PA/Crowd/NitroDrome_PA_Crowd_Cheer_04}}
 {id=WavePlayer_44, properties {soundWaveAssetPtr=/NitroDrome/Assets/Audio/PA/Crowd/NitroDrome_PA_Crowd_Cheer_05}}
 {id=WavePlayer_45, properties {soundWaveAssetPtr=/NitroDrome/Assets/Audio/PA/Crowd/NitroDrome_PA_Crowd_Cheer_06}}
 {id=WavePlayer_46, properties {soundWaveAssetPtr=/NitroDrome/Assets/Audio/PA/Crowd/NitroDrome_PA_Crowd_Cheer_07}}
 {id=WavePlayer_47, properties {soundWaveAssetPtr=/NitroDrome/Assets/Audio/PA/Crowd/NitroDrome_PA_Crowd_Cheer_08}}
 {id=WavePlayer_48, properties {soundWaveAssetPtr=/NitroDrome/Assets/Audio/PA/Crowd/NitroDrome_PA_Crowd_Shock_01}}
 {id=WavePlayer_49, properties {soundWaveAssetPtr=/NitroDrome/Assets/Audio/PA/Crowd/NitroDrome_PA_Crowd_Shock_02}}
 {id=WavePlayer_50, properties {soundWaveAssetPtr=/NitroDrome/Assets/Audio/PA/Crowd/NitroDrome_PA_Crowd_Shock_03}}
 {id=WavePlayer_51, properties {soundWaveAssetPtr=/NitroDrome/Assets/Audio/PA/Crowd/NitroDrome_PA_Crowd_Shock_04}}
 {id=WavePlayer_52, properties {soundWaveAssetPtr=/NitroDrome/Assets/Audio/PA/Crowd/NitroDrome_PA_Crowd_Shock_05}}
 {id=WavePlayer_53, properties {soundWaveAssetPtr=/NitroDrome/Assets/Audio/PA/Crowd/NitroDrome_PA_Crowd_Shock_06}}
]
properties {
 attenuationSettings=/NitroDrome/Assets/Audio/PA/VO/VO_PA_Announcer_Att
 maxDistance=20000
 soundClassObject=&Fort_VO_NPC
 volumeMultiplier=0.85
 concurrencySet [/NitroDrome/Assets/Audio/PA/VO/VO_PA_Announcer_Conc]
}
